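; Creative Score Manager device that awards a configurable, adjustable score to players or teams through minigame events.
; Team-filtered events activate, increment, decrement, reset, toggle or copy player score; activation feeds HUD broadcasts, trigger limits and clamped score updates with output events.
; RepNotifies drive cosmetic-only visuals: centered digit meshes, sign, colored rays and light.
; Audio replicates as a vector pairing a counter with the score, ensuring repeats replicate.
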(event Custom|Activate%20Event
  (params (Instigator :type Controller%20Object%20Reference))
  ;[0.7957 0.005113 1] Gameplay Functions
  (CallFunction|ActivateScoreDevice Instigator true (select (Variables|UserOptions|GetbIncrementOnAwarding) "NewEnumerator1" "NewEnumerator0")))

(event EventBeginPlay
  ;[0.07648 0.25 0.06831] Initialization
  (|CallParentFunction)
  (Effects|RefreshTeamBinding false)
  (Variables|Default|SetCachedMinigameStatsComponent (Minigame|Stats|GetMinigameStatsComponent self)))

(event Custom|Increment%20Event
  (params (Instigator :type Controller%20Object%20Reference))
  ;[0.7957 0.005113 1] Gameplay Functions
  (CallFunction|ActivateScoreDevice Instigator false "NewEnumerator1") ; Activate with no Instigator
)

(event EventInitializeDevice
  ;[0.07648 0.25 0.06831] Initialization
  (|CallParentFunction)
  (CallFunction|ResetDeviceState true)
  (Effects|SetTeamColor)
  (Effects|RefreshTeamBinding false)
  (Class|CreativeUIBroadcastComponent|Initialize (Variables|Default|GetCreative_HUDBroadcast_Component) (Variables|Default|GetFortMinigameProgress) (Variables|UserOptions|Advanced|GetDisplayScoreUpdateOnHUD)))

(event Custom|Decrement%20Event
  (params (Instigator :type Controller%20Object%20Reference))
  ;[0.7957 0.005113 1] Gameplay Functions
  (CallFunction|ActivateScoreDevice Instigator false "NewEnumerator2") ; Activate with no Instigator
)

(event MinigameStateChangedDelegate(FortMinigameProgress)
  (params (Minigame :type Fort%20Minigame%20Object%20Reference) (NewMinigameState :type EFortMinigameState%20Enum :default "PreGame"))
  ;[0.07648 0.25 0.06831] Initialization
  (Variables|Default|SetEnabled (Class|CreativeCommonDeviceFunctionLibrary|IsEnabledPhaseMatch NewMinigameState (Variables|UserOptions|GetEnabledIndex) self))
  (if (Utilities|Enum|Equal(Enum) NewMinigameState "WaitingForCameras")
    (CallFunction|HandleBaseVisibility)))

(event Custom|Enable%20Event
  (params (Instigator :type Controller%20Object%20Reference))
  ;[0.7957 0.005113 1] Gameplay Functions
  (if (CallFunction|IsTeamValid Instigator)
    (Variables|Default|SetEnabled true)))

(event OnPropertiesLoaded(ToyOptionsComponent)
  ;[0.07648 0.25 0.06831] Initialization
  (CallFunction|MigrateUserOptionsifNeeded))

(event Custom|Disable%20Event
  (params (Instigator :type Controller%20Object%20Reference))
  ;[0.7957 0.005113 1] Gameplay Functions
  (if (CallFunction|IsTeamValid Instigator)
    (Variables|Default|SetEnabled false)))

(event EventOnMinigameEnded
  ;[0.07648 0.25 0.06831] Initialization
  (|CallParentFunction)
  (CallFunction|ResetDeviceState true))

(event EventOnAddedToMinigame
  (params (Minigame :type Fort%20Minigame%20Object%20Reference))
  ;[0.07648 0.25 0.06831] Initialization
  (|CallParentFunction Minigame)
  (Effects|RefreshTeamBinding true))

(event EventOnRemovedFromMinigame
  (params (Minigame :type Fort%20Minigame%20Object%20Reference))
  ;[0.07648 0.25 0.06831] Initialization
  (|CallParentFunction Minigame)
  (Effects|RefreshTeamBinding false))

(event Custom|Reset%20Event
  (params (Instigator :type Controller%20Object%20Reference))
  ;[0.7957 0.005113 1] Gameplay Functions
  (if (CallFunction|IsTeamValid Instigator)
    (CallFunction|ResetDeviceState true)))

(event Custom|SetToPlayerScore%20Event
  (params (Instigator :type Controller%20Object%20Reference))
  ;[0.7957 0.005113 1] Gameplay Functions
  (if (CallFunction|IsTeamValid Instigator)
    (CallFunction|SetDeviceScoreToPlayer Instigator)))

(event Custom|Activate%20Score%20Device
  (params (Instigator :type Controller%20Object%20Reference) (AwardScore :type Boolean) (Update%20Score :type EAwardScoreType%20Enum :default "NewEnumerator0"))
  ;[0.1719 0.6901 1 0.2] Score Logic
  (if (and (CallFunction|IsTeamValid Instigator) (Variables|Default|GetEnabled))
    (bind _asfort_player_controller (Utilities|Casting|CastToFortPlayerController Instigator)
      (:then
        _)
      (:CastFailed
        _))
    (Variables|Default|SetLast%20Instigator _asfort_player_controller)
    (if AwardScore
      (Class|CreativeCommonDeviceFunctionLibrary|AwardScore self (Variables|Default|GetNextScore) (Variables|UserOptions|GetSetScore) (Variables|Default|GetLast%20Instigator) (Variables|Default|GetFortMinigameProgress))
      (if (!= (Variables|Default|GetNextScore) 0)
        _
        (elif (Variables|UserOptions|Advanced|GetDisplayScoreUpdateIfScoreIs0)
          _
          (else
            (CallFunction|RepAudioFX)
            (CallFunction|IncrementTimesTriggered)
            (CallFunction|UpdateScoreValue Update%20Score (Variables|UserOptions|GetSubsequentScore)))))
      (Class|CreativeHUDBroadcastComponent|DoScoreEvent (Variables|Default|GetCreative_HUDBroadcast_Component) (Variables|Default|GetLast%20Instigator) (Variables|Default|GetNextScore) (Variables|UserOptions|Advanced|GetDisplayScoreUpdateOnHUD))
      (else
        _))))

(fn ActivateFromOtherDevice
  (params (Instigator :type Controller%20Object%20Reference) (Requester :type ERequesterType%20Enum :default "NewEnumerator0"))
  (CallFunction|ActivateScoreDevice Instigator true (select (Variables|UserOptions|GetbIncrementOnAwarding) "NewEnumerator1" "NewEnumerator0"))
  (return true))

(fn CacheMaterials
  (Networking|CanExecuteCosmeticEvents
    (:True
      (Variables|MaterialRefs|SetSign%20MID (Material|GetMIDfromMaterial (Variables|Default|GetSign) 0 (Variables|Effects|GetDigit%20Material)))
      (Variables|MaterialRefs|SetDigit1%20MID (Material|GetMIDfromMaterial (Variables|Default|GetDigit1) 0 (Variables|Effects|GetDigit%20Material)))
      (Variables|MaterialRefs|SetDigit2%20MID (Material|GetMIDfromMaterial (Variables|Default|GetDigit2) 0 (Variables|Effects|GetDigit%20Material)))
      (Variables|MaterialRefs|SetDigit3%20MID (Material|GetMIDfromMaterial (Variables|Default|GetDigit3) 0 (Variables|Effects|GetDigit%20Material)))
      (Variables|MaterialRefs|SetLightRays%20MID (Material|GetMID (Variables|Default|GetLightRays))))))

(fn CanInteract
  (params (InteractingPawn :type Fort%20Pawn%20Object%20Reference) (InteractionBeingAttempted :type EInteractionBeingAttempted%20Enum :default "FirstInteraction") (InteractionType :type TInteractionType%20Enum :default "IT_NoInteraction"))
  (return false))

(fn CreateOrReattachScoreColorMIDs
  ;[1 1 1] Reattach cached original materials for MIDs which use score color
  (Material|ReattachOrCreateMID (Variables|Default|GetSign) (Variables|MaterialRefs|GetSign%20MID))
  (Material|ReattachOrCreateMID (Variables|Default|GetDigit1) (Variables|MaterialRefs|GetDigit1%20MID))
  (Material|ReattachOrCreateMID (Variables|Default|GetDigit2) (Variables|MaterialRefs|GetDigit2%20MID))
  (Material|ReattachOrCreateMID (Variables|Default|GetDigit3) (Variables|MaterialRefs|GetDigit3%20MID))
  (Material|ReattachOrCreateMID (Variables|Default|GetLightRays) (Variables|MaterialRefs|GetLightRays%20MID)))

(fn GetDeviceColor
  (return 0))

(fn GetInstigatingController
  (return 0))

(fn GetSpawnedPickups
  (return 0))

(fn GetVisibilityComponentsWithCollision
  (return (Utilities|Map|MakeMap (Variables|Debug|GetStaticMeshComponent) "QueryAndPhysics" (Variables|Default|GetDigit1) "NoCollision" (Variables|Default|GetSign) "NoCollision" (Variables|Default|GetLightRays) "NoCollision" (Variables|Default|GetPointLight) "NoCollision" (Variables|Default|GetAudio) "NoCollision" (Variables|Default|GetDigit3) "NoCollision")))

(fn HandleBaseVisibility
  (LocalVariables|SetBaseIsVisible (== (Variables|UserOptions|GetSetVisibilityKnob) 1))
  (Rendering|SetVisibility (Variables|Debug|GetStaticMeshComponent) (LocalVariables|GetBaseIsVisible) false)
  (Collision|SetCollisionEnabled (Variables|Debug|GetStaticMeshComponent) (select (LocalVariables|GetBaseIsVisible) "QueryAndPhysics" "NoCollision")))

(fn IncrementTimesTriggered
  (bind _output_get (Variables|Default|SetTimesTriggered (+ (Variables|Default|GetTimesTriggered) 1)))
  (if (and (== _output_get (Variables|UserOptions|GetTimesCanTrigger)) (> (Variables|UserOptions|GetTimesCanTrigger) 0))
    (GameplayEvent|BroadcastEvent (Variables|UserOptions|Events|GetOn%20Max%20Triggers) 0 (Utilities|Struct|MakeGameplayEventonMaxTriggers (Variables|Default|GetLast%20Instigator)))
    (Variables|Default|SetEnabled false)))

(fn IsEnabled
  (return false))

(fn IsHoldingItemForPawn
  (params (Item :type Item%20Variant%20Handle%20Structure) (Pawn :type Pawn%20Object%20Reference))
  (return false))

(fn IsTeamValid
  (params (Actor :type Actor%20Object%20Reference))
  (return (or (Knobs|Team|IsTeamAny (Variables|UserOptions|GetTeamSetting)) (== (Knobs|Team|GetTeamValueasIndex (Variables|UserOptions|GetTeamSetting)) (Math|Conversions|ToInteger(Byte) (Team|GetActorPvPTeamIndex Actor))))))

(fn IsVisibleDuringMinigame
  (return (!= (Variables|UserOptions|GetSetVisibilityKnob) 0)))

(fn MessageScore
  (return (Variables|UserOptions|Advanced|GetHUD%20Message) (Variables|UserOptions|Advanced|GetHUD%20MessageScoreColor) (Variables|UserOptions|Advanced|GetHUD%20MessageColor) (Variables|UserOptions|Advanced|GetResetHUD%20MessageScore)))

(fn MigrateUserOptionsIfNeeded
  (bind _toyoptionscomponent (Variables|Default|GetToyOptionsComponent))
  (bind _currentoptions (OptionData|GetPropertyOverrides _toyoptionscomponent))
  (bind _oldvisibilityknob (LocalVariables|GetOldVisibilityKnob))
  (bind (_value _returnvalue) (Utilities|Map|Find _currentoptions _oldvisibilityknob))
  (bind _toyoptionscomponent_2 (Variables|Default|GetToyOptionsComponent))
  (bind _toyoptionscomponent_3 (Variables|Default|GetToyOptionsComponent))
  (if _returnvalue
    (if (Utilities|String|Equal,CaseInsensitive(String) _value "True")
      (OptionData|SetPropertyOverrides _toyoptionscomponent_3 (Utilities|Map|MakeMap (LocalVariables|GetNewVisibilityKnob) "Visible"))
      (OptionData|SetPropertyOverrides _toyoptionscomponent_3 (Utilities|Map|MakeMap (LocalVariables|GetNewPlayAudioKnob) "On"))
      (else
        (OptionData|SetPropertyOverrides _toyoptionscomponent_2 (Utilities|Map|MakeMap (LocalVariables|GetNewVisibilityKnob) "NotVisible"))
        (OptionData|SetPropertyOverrides _toyoptionscomponent_2 (Utilities|Map|MakeMap (LocalVariables|GetNewPlayAudioKnob) "Off"))))
    (OptionData|RemovePropertyOverrides (Variables|Default|GetToyOptionsComponent) (Utilities|Array|MakeArray (LocalVariables|GetOldVisibilityKnob)))))

(fn OnRep_AudioEvents
  (bind (_audio_events_x _audio_events_y) (Math|Vector2D|BreakVector2D (Variables|GetAudio%20Events_X)))
  (bind _audio (Variables|Default|GetAudio))
  ;[1 0.759 0] Audio Event Y passes through the score that was just triggered. Use this to play an audio sample on the client
  (Networking|CanExecuteCosmeticEvents
    (:True
      (if (Math|Boolean|ANDBoolean (Default|IsinPlayMode) (Variables|Default|GetEnabled) (== (Variables|UserOptions|GetPlayAudioKnob) 1))
        (Audio|Components|Audio|SetSound _audio (select (>= _audio_events_y 0.0) (Variables|Default|GetAdd%20Audio) (Variables|Default|GetSubtract%20Audio)))
        (Audio|Components|Audio|Play _audio 0.0)))))

(fn OnRep_Enabled
  (Networking|CanExecuteCosmeticEvents
    (:True
      (Effects|UpdateOutputVisibility))
    (:False
      (if (not (Variables|Default|GetEnabled))
        (Variables|Default|SetLast%20Instigator 0)))))

(fn OnRep_NextScore
  (Effects|UpdateReadout (Variables|Default|GetNextScore)))

(fn OnRep_SetScore
  ;[1 0.5253 0 0.5] When the Score Award Type changes, update the readout (as score award type is one of the things readout displays)
  (Effects|UpdateReadout (Variables|Default|GetNextScore)))

(fn RefreshTeamBinding
  (params (Unbind :type Boolean))
  (Networking|CanExecuteCosmeticEvents
    (:True
      (Teams|RefreshTeamChangeBinding (Variables|Default|GetFortMinigameProgress) Unbind (EventDispatchers|CreateEvent :!event Set%20Team%20Color)))))

(fn RepAudioFX
  (bind _audio_events_x (Math|Vector2D|BreakVector2D (Variables|GetAudio%20Events_X)))
  (Variables|SetAudio%20Events_X (Math|Vector2D|MakeVector2D (+ _audio_events_x 1.0) (Math|Conversions|ToFloat(Integer) (Utilities|Select :Index (Variables|UserOptions|GetSetScore) :NewEnumerator0 0 :NewEnumerator1 (Variables|Default|GetNextScore) :NewEnumerator2 (* (Variables|Default|GetNextScore) -1) :NewEnumerator3 (Variables|Default|GetNextScore))))) ; Sends amount of score and an incrementing number (to ensure replication)
)

(fn ResetDeviceState
  (params (Enable :type Boolean))
  (Networking|SwitchHasAuthority
    (:Authority
      (Variables|Default|SetTimesTriggered 0)
      (CallFunction|UpdateScoreValue "NewEnumerator3" (Variables|UserOptions|GetScoreValueSetting))
      (Variables|Default|SetEnabled Enable))))

(fn SetDeviceScoretoPlayer
  (params (Sending%20Controller :type Controller%20Object%20Reference))
  (bind _playerstate (Variables|/Script/Engine.Controller|GetPlayerState Sending%20Controller)
    (:then
      (bind _asfort_player_state (Utilities|Casting|CastToFortPlayerState _playerstate))
      (bind _cachedminigamestatscomponent (Variables|Default|GetCachedMinigameStatsComponent)
        (:then
          (CallFunction|UpdateScoreValue "NewEnumerator3" (Minigame|Stats|GetStatforPlayer _cachedminigamestatscomponent _asfort_player_state (Class|FortMinigameStatsComponent|GetLoadedStatfromClass _cachedminigamestatscomponent "/Game/Athena/Playset/Minigames/StatFilters/MinigameStat_Score.MinigameStat_Score_C") false)))
        (:else)))
    (:else)))

(fn SetTeamColor
  (Networking|CanExecuteCosmeticEvents
    (:True
      (Teams|UpdateTeamColor (Variables|Default|GetFortMinigameProgress) (Variables|Debug|GetStaticMeshComponent) (select (Knobs|Team|IsTeamAny (Variables|UserOptions|GetTeamSetting)) -1 (Knobs|Team|GetTeamValueasIndex (Variables|UserOptions|GetTeamSetting)))))))

(fn SetTriggerCollision
  (params (On :type Boolean))
  (return false))

(fn UpdateOutputVisibility
  (Networking|CanExecuteCosmeticEvents
    (:True
      (bind _output_get (LocalVariables|Setlocal%20Visibility (Math|Boolean|ANDBoolean (or (< (Variables|Default|GetTimesTriggered) (Variables|UserOptions|GetTimesCanTrigger)) (== (Variables|UserOptions|GetTimesCanTrigger) -1)) (Variables|Default|GetEnabled) (VisibleInGameComponent|GetDesiredVisibility))))
      (Development|SetHiddeninGame (Variables|Default|GetLightRays) (not _output_get) true)
      (Effects|UpdateReadout (Variables|Default|GetNextScore)))))

(fn UpdateReadout
  (params (Number :type Integer))
  (bind _local_value (LocalVariables|Getlocal%20Value))
  (bind _returnvalue (Math|Integer|%25%28Integer%29 _local_value 10))
  (bind _digit1 (Variables|Default|GetDigit1))
  (bind _digit2 (Variables|Default|GetDigit2))
  (bind _digit3 (Variables|Default|GetDigit3))
  (bind _sign (Variables|Default|GetSign))
  (Networking|CanExecuteCosmeticEvents
    (:True
      (LocalVariables|Setlocal%20Sign (* (Math|Integer|Sign(Integer) Number) (select (Utilities|Enum|Equal(Enum) (Variables|UserOptions|GetSetScore) "NewEnumerator2") -1 1)))
      (LocalVariables|Setlocal%20Value (Math|Integer|Absolute(Integer) Number))
      (for _array_element (Utilities|Array|MakeArray (Variables|Default|GetDigit3) (Variables|Default|GetDigit2) (Variables|Default|GetDigit1) (Variables|Default|GetSign))
        (UEFNConversionHelper|SetComponentHiddenEditor _array_element true))
      (if (and (VisibleInGameComponent|GetDesiredVisibility) (Variables|Default|GetEnabled))
        ;[0 0.3855 1] Enable or Disable an appropriate number of digits.
        (for _index (range (+ (Math|Integer|SelectInt (Math|Integer|SelectInt 2 1 (> (LocalVariables|Getlocal%20Value) 99)) 0 (> (LocalVariables|Getlocal%20Value) 9)) 1))
          (Utilities|Array|Add (LocalVariables|Getlocal%20Digits) (Utilities|Select :Index _index :Option%200 _digit1 :Option%201 _digit2 :Option%202 _digit3))
          (UEFNConversionHelper|SetComponentHiddenEditor (Utilities|Select :Index _index :Option%200 _digit1 :Option%201 _digit2 :Option%202 _digit3) (not (VisibleInGameComponent|GetDesiredVisibility))))
        ;[1 0.6061 0] Set Meshes
        (bind (_arrayindex _arrayelement) (Utilities|Array|ReverseforEachLoop (LocalVariables|Getlocal%20Digits))
          (:LoopBody
            (Components|StaticMesh|SetStaticMesh _arrayelement (Utilities|Array|Get(acopy) (Variables|Effects|GetDigitsStaticMeshTemplates) _returnvalue))
            (LocalVariables|Setlocal%20Value (/ (- (LocalVariables|Getlocal%20Value) _returnvalue) 10)))
          (:Completed
            ;[1 1 1] Insert the Sign at the start of the array, assuming we're accumulating rather than setting score, and there is score to add/remove.
            (if (and (!= (LocalVariables|Getlocal%20Sign) 0) (or (Utilities|Enum|NotEqual(Enum) (Variables|UserOptions|GetSetScore) "NewEnumerator3") (Utilities|Enum|Equal(Enum) (Variables|UserOptions|GetSetScore) "NewEnumerator0")))
              (Components|StaticMesh|SetStaticMesh (Variables|Default|GetSign) (Utilities|Array|Get(acopy) (Variables|Effects|GetSignStaticMeshes) (Math|Integer|Max(Integer) (LocalVariables|Getlocal%20Sign))))
              (Utilities|Array|Insert (LocalVariables|Getlocal%20Digits) _sign)
              (UEFNConversionHelper|SetComponentHiddenEditor _sign (not (VisibleInGameComponent|GetDesiredVisibility)))
              (else
                _))
            ;[0 0.3855 1] Center Digits
            (LocalVariables|Setlocal%20Start (* (Utilities|Array|LastIndex (LocalVariables|Getlocal%20Digits)) (/ (Variables|Effects|GetDigit%20Kerning) -2.0)))
            (for (_array_element_2 _array_index) (LocalVariables|Getlocal%20Digits)
              (Transformation|SetRelativeLocation _array_element_2 (Math|Vector|MakeVector 0.0 (+ (* _array_index (Variables|Effects|GetDigit%20Kerning)) (LocalVariables|Getlocal%20Start))) false false))
            (Effects|UpdateScoreColor)))))))

(fn UpdateScoreColor
  (bind _lightrays (Variables|Default|GetLightRays))
  (bind _mid (Material|GetMID _lightrays))
  (Networking|CanExecuteCosmeticEvents
    (:True
      (CallFunction|CreateorReattachScoreColorMIDs)
      (for _index (range (+ (Utilities|Array|LastIndex (Variables|Effects|GetColorPositive)) 1))
        (LocalVariables|Seti _index)
        (LocalVariables|Setlocal%20Color (Utilities|Select :Index (Variables|UserOptions|GetSetScore) :NewEnumerator0 (select (>= (Variables|Default|GetNextScore) 0) (Utilities|Array|Get(acopy) (Variables|Effects|GetColorPositive) (LocalVariables|Geti)) (Utilities|Array|Get(acopy) (Variables|Effects|GetColorNegative) (LocalVariables|Geti))) :NewEnumerator1 (select (>= (Variables|Default|GetNextScore) 0) (Utilities|Array|Get(acopy) (Variables|Effects|GetColorPositive) (LocalVariables|Geti)) (Utilities|Array|Get(acopy) (Variables|Effects|GetColorNegative) (LocalVariables|Geti))) :NewEnumerator2 (select (>= (Variables|Default|GetNextScore) 0) (Utilities|Array|Get(acopy) (Variables|Effects|GetColorNegative) (LocalVariables|Geti)) (Utilities|Array|Get(acopy) (Variables|Effects|GetColorPositive) (LocalVariables|Geti))) :NewEnumerator3 (select (>= (Variables|Default|GetNextScore) 0) (Utilities|Array|Get(acopy) (Variables|Effects|GetColorPositiveSet) (LocalVariables|Geti)) (Utilities|Array|Get(acopy) (Variables|Effects|GetColorNegativeSet) (LocalVariables|Geti)))))
        (switch int (LocalVariables|Geti)
          (:0
            (Rendering|Material|SetVectorParameterValue (Material|GetMIDfromMaterial (Variables|Default|GetDigit1) 0 (Variables|Effects|GetDigit%20Material)) "ColorInner" (LocalVariables|Getlocal%20Color) :!class /Script/Engine.MaterialInstanceDynamic))
          (:1
            (Rendering|Material|SetVectorParameterValue _mid "ColorHigh" (LocalVariables|Getlocal%20Color) :!class /Script/Engine.MaterialInstanceDynamic))
          (:2
            (Rendering|Material|SetVectorParameterValue _mid "ColorLow" (LocalVariables|Getlocal%20Color) :!class /Script/Engine.MaterialInstanceDynamic))
          (:3
            (Rendering|Components|Light|SetLightColor (Variables|Default|GetPointLight) (LocalVariables|Getlocal%20Color) true)))))))

(fn UpdateScoreValue
  (params (UpdateScore :type EAwardScoreType%20Enum :default "NewEnumerator0") (Score :type Integer))
  (Variables|Default|SetNextScore (Math|Integer|Clamp(Integer) (Math|Integer|SelectInt Score (+ (Variables|Default|GetNextScore) (Utilities|Select :Index (Variables|GetUpdateScore) :NewEnumerator0 0 :NewEnumerator1 (Variables|GetScore) :NewEnumerator2 (* (Variables|GetScore) -1) :NewEnumerator3 0)) (Utilities|Enum|Equal(Enum) UpdateScore "NewEnumerator3")) (Variables|UserOptions|GetMinScore) (Variables|UserOptions|GetMaxScore)))
  (if (== (Variables|Default|GetNextScore) (Variables|Default|GetTransmitOnScore))
    (GameplayEvent|BroadcastEvent (Variables|UserOptions|Events|GetOn%20Score%20Output) 0 (Utilities|Struct|MakeGameplayEventonScoreOutput (Variables|Default|GetLast%20Instigator)))))

(fn ConstructionScript
  (CallFunction|CacheMaterials)
  (|CallParentFunction))
